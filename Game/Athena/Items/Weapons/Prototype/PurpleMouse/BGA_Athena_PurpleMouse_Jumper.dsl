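; PurpleMouse jumping proximity mine actor that aligns to its landing surface, arms after a delay with team-colored effects, and allows pickup before jumping.
; Server-side overlap spheres drive beeping and line-of-sight-checked triggering, using a gameplay effect on occluded targets; damage or water also triggers it.
; It launches upward on a spin timeline, where shots knock it away and reset the timer.
; The timeline ends in an explosion applying damage effects and optionally marking pawns.
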(event Custom|SetMyTeam
  (params (Team :type Byte))
  ;[1 1 1] Set Team
  (Networking|SwitchHasAuthority
    (:Authority
      (bind _output_get (Variables|Default|SetMyTeam Team))
      (Variables|Default|SetTeamHasBeenSet true)
      (Development|PrintString (Utilities|String|Append "Set team to " (Utilities|String|ToString(Byte) _output_get)) true true "(R=1.000000,G=0.754107,B=0.000000,A=1.000000)" 5.0))))

(event Custom|OnInstigatorDied
  (params (DamagedActor :type Actor%20Object%20Reference) (Damage :type Float%20%28single-precision%29) (InstigatedBy :type Controller%20Object%20Reference) (DamageCauser :type Actor%20Object%20Reference) (HitLocation :type Vector) (FHitComponent :type Primitive%20Component%20Object%20Reference) (BoneName :type Name) (Momentum :type Vector))
  ;[1 1 1] OnInstigator Died Kill this mine
  (Actor|SetLifeSpan 0.01))

(event EventBeginPlay
  (bind _beepparticles (Variables|Default|GetBeepParticles))
  (|CallParentFunction)
  (Utilities|FlowControl|Sequence
    (:then_0
      (Development|SetHiddeninGame (Variables|Default|GetSM_Cylinder_Beam) true false)
      (bind _asfort_pawn (Utilities|Casting|CastToFortPawn (Variables|Actor|GetInstigator)))
      (if (not (FortPawn|IsDead _asfort_pawn))
        (CallFunction|SetMyTeam (Team|GetActorTeam _asfort_pawn))))
    (:then_1
      ;[1 1 1] OnInstigator Died Kill this mine
      (if (Fortnite|IsCreativeorPlaygroundModePlaylist)
        (bind _asfort_player_pawn_athena (Utilities|Casting|CastToFortPlayerPawnAthena (Variables|Actor|GetInstigator)))
        (EventDispatchers|BindEventtoOnDied _asfort_player_pawn_athena (delegate Custom|OnInstigatorDied))))
    (:then_2
      (EventDispatchers|BindEventtoOnDied (delegate Custom|Destroyed)))
    (:then_3
      (Fortnite|ShowSoundIndicatorLocal :Location (Transformation|GetActorLocation) :MaxAudibleDistance (Variables|/Script/Engine.SoundBase|GetMaxDistance (Variables|Default|GetBeepSound)) :IndicatorType "Generic" :Instigator self :Tint "(R=0.000000,G=0.000000,B=0.000000,A=0.000000)" :OverrideIcon "/Game/UI/Foundation/Textures/Icons/HUD/T-Icon-Alert2-32.T-Icon-Alert2-32" :AttenuationMaxRangeMultiplier 1.0 :AttenuationIntensityMultiplier 1.0)
      ;[0 0.4521 1] Impact Dust FX
      (Effects|SpawnEmitteratLocation "/Game/Athena/Items/Consumables/C4/P_Athena_C4_Land_Dust.P_Athena_C4_Land_Dust" (Variables|/Script/Engine.SceneComponent|GetRelativeLocation (Variables|Default|GetStaticMesh)) (Math|Conversions|RotationFromXVector (Math|Vector|GetUpVector (Variables|/Script/Engine.SceneComponent|GetRelativeRotation (Variables|Default|GetStaticMesh)))))
      ;[1 1 1] Initialize Team Colors
      (Utilities|FlowControl|Delay (ScalableFloat|GetValueatLevel (Variables|Default|GetTimeToArm)))
      (Development|PrintString "Armed" true true "(R=1.000000,G=0.000000,B=0.824123,A=1.000000)" 5.0)
      (Utilities|FlowControl|Sequence
        (:then_0
          (bind _returnvalue (Effects|SpawnEmitterAttached :EmitterTemplate (select (Utilities|SameTeam) "/Game/Athena/Items/Weapons/Prototype/PurpleMouse/P_PurpleMouse_AlertLoop_Friendly.P_PurpleMouse_AlertLoop_Friendly" "/Game/Athena/Items/Weapons/Prototype/PurpleMouse/P_PurpleMouse_AlertLoop.P_PurpleMouse_AlertLoop") :AttachToComponent (Variables|Default|GetVisibleMesh) :AttachPointName "None" :Location "0, 0,8.000000" :Scale "1.000000,1.000000,1.000000" :LocationType "SnapToTarget" :bAutoDestroy false :PoolingMethod "None" :bAutoActivate true))
          (Variables|Default|SetBeepParticles _returnvalue)
          (Networking|CanExecuteCosmeticEvents
            (:True
              (Effects|Components|ParticleSystem|SetVectorParameter (Variables|Default|GetBeepParticles) "Color" (Math|Conversions|ToVector(LinearColor) (select (Utilities|SameTeam) (Variables|Default|GetFriendlyParticleColor) (Variables|Default|GetEnemyParticleColor))))
              (Rendering|Material|SetVectorParameterValueonMaterials (Variables|Default|GetVisibleMesh) "Emissive" (Math|Conversions|ToVector(LinearColor) (select (Utilities|SameTeam) (Variables|Default|GetFriendlyEmissiveColor) (Variables|Default|GetEnemyEmissiveColor))))
              (Rendering|Material|SetMaterial (Variables|Default|GetSM_Cylinder_Beam) 0 (Selection|SelectMaterialInstance "/Game/Athena/Items/Weapons/Prototype/PurpleMouse/VFX/MI_Beam_PurpleMouse_Friendly.MI_Beam_PurpleMouse_Friendly" "/Game/Athena/Items/Weapons/Prototype/PurpleMouse/VFX/MI_Beam_PurpleMouse.MI_Beam_PurpleMouse" (Utilities|SameTeam))))))
        (:then_1
          (Components|Sphere|SetSphereRadius (Variables|Default|GetBeepBlinkTrigger) (* (ScalableFloat|GetValueatLevel (Variables|Default|GetBeepTriggerRadius)) 0.5) true) ; NOTE: scale of root is set to 2 when BGA is spawned, so this radius is actually doubled.
          (Components|Sphere|SetSphereRadius (Variables|Default|GetExplosionTrigger) (* (ScalableFloat|GetValueatLevel (Variables|Default|GetTriggerRadius)) 0.5) true) ; NOTE: scale of root is set to 2 when BGA is spawned, so this radius is actually doubled.
        )
        (:then_2
          (Effects|Components|ParticleSystem|SetEmitterEnable _beepparticles "Static" true)
          (Effects|Components|ParticleSystem|SetEmitterEnable _beepparticles "Blink" false)
          (Development|SetHiddeninGame (Variables|Default|GetSM_Cylinder_Beam) false false)
          (if (Class|AthenaFunctionLibrary|RowtoBool (Variables|Default|GetPlayBeepSoundOnArm))
            (Audio|PlaySoundatLocation "/Game/Sounds/Fort_Gadgets/ProximityMine2/ProximityMine_Beep_Cue.ProximityMine_Beep_Cue" (Transformation|GetActorLocation))))
        (:then_3
          (if (<= (ScalableFloat|GetValueatLevel (Variables|Default|GetTimeToArm)) 0.1)
            ;[1 1 1] Arming after a Delay (need *at least* a 0.1 delay before turning on collision to avoid bugs)
            (Utilities|FlowControl|Delay 0.1)
            (else
              _))
          (Networking|SwitchHasAuthority
            (:Authority
              (Collision|SetCollisionEnabled (Variables|Default|GetExplosionTrigger) "QueryOnly") ; Explosion Trigger isn't armed right away, gives enemy extra reaction time and prevents bugs
              (Collision|SetCollisionEnabled (Variables|Default|GetBeepBlinkTrigger) "QueryOnly") ; Explosion Trigger isn't armed right away, gives enemy extra reaction time and prevents bugs
            )))
        (:then_4
          ;[1 1 1] Shoot it after it jumped to make if fly away
          (EventDispatchers|BindEventtoOnDamaged (delegate Custom|Damaged)))))))

(event Custom|Destroyed
  (params (DamagedActor :type Actor%20Object%20Reference) (Damage :type Float%20%28single-precision%29) (InstigatedBy :type Controller%20Object%20Reference) (DamageCauser :type Actor%20Object%20Reference) (HitLocation :type Vector) (FHitComponent :type Primitive%20Component%20Object%20Reference) (BoneName :type Name) (Momentum :type Vector))
  ;[1 1 1] On Destroyed
  (Audio|Components|Audio|FadeOut (Variables|Default|GetAmbientHumAudio) 0.5 0.0 "Linear" :!class /Script/Engine.AudioComponent)
  (if (not (Variables|Default|GetIsExlode?))
    ;[1 1 1] Pick-Up effects
    (Effects|Components|ParticleSystem|SpawnFXAtLocation "/Game/Athena/Items/Weapons/Prototype/Badger_Bangs/FX/NS_BlackMonday_PrjTakeBack.NS_BlackMonday_PrjTakeBack" (Transformation|GetActorLocation) (Math|Conversions|RotationFromXVector (Variables|Default|GetGravHitNormal)))
    (Audio|PlaySoundatLocation (Variables|Default|GetPickUpSound) (Transformation|GetActorLocation)))
  (Rendering|SetActorHiddenInGame true)
  (Collision|SetActorEnableCollision false))

(event Custom|Damaged
  (params (DamagedActor :type Actor%20Object%20Reference) (Damage :type Float%20%28single-precision%29) (InstigatedBy :type Controller%20Object%20Reference) (DamageCauser :type Actor%20Object%20Reference) (HitLocation :type Vector) (FHitComponent :type Primitive%20Component%20Object%20Reference) (BoneName :type Name) (Momentum :type Vector))
  (bind (_asbga_athena_purple_mouse_jumper _bsuccess) (Utilities|Casting|CastToBGA_Athena_PurpleMouse_Jumper DamageCauser))
  ;[1 1 1] Shoot it after it jumped to make if fly away
  (Networking|SwitchHasAuthority
    (:Authority
      (if (and (Variables|Default|GetJumped) (not _bsuccess))
        (CallFunction|PushBack Momentum)))))

(event Custom|PushBack
  (params (V :type Vector))
  ;[1 1 1] Shoot it after it jumped to make if fly away
  (CallFunction|RestartSimulation)
  (Variables|/Script/Engine.MovementComponent|SetVelocity (* (+ (Math|Vector|Normalize (Math|Vector|ProjectVectorontoPlane V "0, 0,1.000000")) "0, 0,0.500000") 1.2e+03) (Variables|Default|GetProjectileMovement))
  (CallFunction|ResetExplosionTimer)
  (Collision|SetActorEnableCollision true))

(event OnComponentBeginOverlap(ExplosionTrigger)
  (params (OverlappedComponent :type Primitive%20Component%20Object%20Reference) (OtherActor :type Actor%20Object%20Reference) (OtherComp :type Primitive%20Component%20Object%20Reference) (OtherBodyIndex :type Integer) (bFromSweep :type Boolean) (SweepResult :type Hit%20Result%20Structure%20%28by%20ref%29))
  ;[1 1 1] Explosion Triggers
  (Networking|SwitchHasAuthority
    (:Authority
      (Utilities|CheckforExplosionTrigger OtherActor
        (:Trigger
          (Utilities|CheckLOS OtherActor 3e+01
            (:LOS
              (CallFunction|JumpIntoExplosion))
            (:No%20LOS
              (GameplayEffects|ApplyGameplayEffectToTarget :self (Ability|GetAbilitySystemComponent self) :GameplayEffectClass "/Game/Athena/Items/Weapons/Prototype/PurpleMouse/GE_PurpleMouse_ApplyLOSChecker.GE_PurpleMouse_ApplyLOSChecker_C" :Target (Ability|GetAbilitySystemComponent OtherActor) :Level 0.0))))))))

(event OnComponentEndOverlap(ExplosionTrigger)
  (params (OverlappedComponent :type Primitive%20Component%20Object%20Reference) (OtherActor :type Actor%20Object%20Reference) (OtherComp :type Primitive%20Component%20Object%20Reference) (OtherBodyIndex :type Integer))
  ;[1 1 1] Explosion Triggers
  (GameplayEffects|RemoveActiveEffectswithTags (Ability|GetAbilitySystemComponent OtherActor) "(GameplayTags=((TagName=\"Item.Consumable.PurpleMouse.LOSChecker\")))"))

(event EventOnDamageServer
  (params (Damage :type Float%20%28single-precision%29) (DamageTags :type Gameplay%20Tag%20Container%20Structure%20%28by%20ref%29) (Momentum :type Vector :default "0, 0, 0") (HitInfo :type Hit%20Result%20Structure%20%28by%20ref%29) (InstigatedBy :type Controller%20Object%20Reference) (DamageCauser :type Actor%20Object%20Reference) (EffectContext :type Gameplay%20Effect%20Context%20Handle%20Structure))
  ;[1 1 1] Explosion Triggers
  (CallFunction|JumpIntoExplosion))

(event EventStopSim
  (params (Hit :type Hit%20Result%20Structure))
  (bind (_bblockinghit _binitialoverlap _time _distance _location _impactpoint _normal) (Collision|BreakHitResult Hit))
  (|CallParentFunction Hit)
  (if (and (not (Variables|Default|GetJumped)) (> (Math|Vector|VectorLength _normal) 0.0))
    (Transformation|SetActorRotation (Math|Rotator|MakeRotfromZ _normal) true)
    (Transformation|SetActorLocation _impactpoint false true)))

(event OnEnterWaterEvent(FortWaterInteraction)
  (params (WaterBody :type Fort%20Water%20Body%20Actor%20Object%20Reference) (WaterInteractionComponent :type Fort%20Water%20Interaction%20Component%20Object%20Reference) (bIsFirstBody :type Boolean))
  ;[1 1 1] Explosion Triggers
  (CallFunction|JumpIntoExplosion))

(event OnComponentBeginOverlap(BeepBlinkTrigger)
  (params (OverlappedComponent :type Primitive%20Component%20Object%20Reference) (OtherActor :type Actor%20Object%20Reference) (OtherComp :type Primitive%20Component%20Object%20Reference) (OtherBodyIndex :type Integer) (bFromSweep :type Boolean) (SweepResult :type Hit%20Result%20Structure%20%28by%20ref%29))
  ;[1 1 1] Beep Logic
  (Networking|SwitchHasAuthority
    (:Authority
      (if (not (Math|Byte|Equal(Byte) (Team|GetActorTeam (Utilities|Casting|CastToFortPlayerPawnAthena OtherActor)) (Variables|Default|GetMyTeam)))
        (Utilities|Array|AddUnique (Variables|Default|GetActorsInBeepRadius) OtherActor)
        (CallFunction|TriggerBeep)))))

(event Custom|JumpIntoExplosion
  ;[1 1 1] Jump
  (if (not (Variables|Default|GetIsExlode?))
    (Variables|Default|SetIsExlode? true)
    (Development|SetHiddeninGame (Variables|Default|GetSM_Cylinder_Beam) true false)
    (CallFunction|DetachandLaunch)
    (CallFunction|TriggerJumpTimeline)
    (Networking|CanExecuteCosmeticEvents
      (:True
        (Effects|SpawnEmitteratLocation "/Game/Athena/Items/Weapons/Prototype/PurpleMouse/P_PurpleMouse_Poof.P_PurpleMouse_Poof" (+ (Transformation|GetActorLocation) "0, 0,10.000000") (Math|Conversions|RotationFromXVector (Variables|Default|GetGravHitNormal)) "1.400000,1.400000,1.400000")))))

(event Custom|Explode
  (bind _instigator (Variables|Actor|GetInstigator))
  (bind _returnvalue (Fortnite|GetFortPlayerControllerfromActor _instigator))
  (bind _instigator_2 (Variables|Actor|GetInstigator))
  (bind _returnvalue_2 (Ability|GetFortAbilitySystemComponent _instigator_2))
  (bind _markduration (Variables|Default|GetMarkDuration))
  (bind _returnvalue_3 (ScalableFloat|GetValueatLevel _markduration))
  (bind _returnvalue_4 (Transformation|GetActorLocation))
  (bind _asfort_player_controller_athena (Utilities|Casting|CastToFortPlayerControllerAthena _returnvalue))
  ;[1 1 1] Explode
  (Utilities|FlowControl|Sequence
    (:then_0
      (Components|DestroyComponent (Variables|Default|GetBeepParticles))
      (Collision|SetActorEnableCollision false)
      (Actor|SetLifeSpan 0.2)
      (Audio|PlaySoundatLocation (Variables|Default|GetExplodeSound) _returnvalue_4 "0, 0, 0" 1.0 1.0 0.0 0 0 self)
      (Effects|SpawnEmitteratLocation "/Game/Athena/Items/Weapons/Prototype/PurpleMouse/VFX/P_PurpleMouse_Exp.P_PurpleMouse_Exp" _returnvalue_4)
      (GameplayCue|ExecuteGameplayCue%28Burst%29Local self (Variables|Default|GetCameraShakeGC) (Ability|GameplayCue|MakeGameplayCueParameters :NormalizedMagnitude 1.0 :RawMagnitude 3.0 :Location _returnvalue_4 :Normal "0,0,0" :GameplayEffectLevel 1 :AbilityLevel 1 :bReplicateLocationWhenUsingMinimalRepProxy false))
      (Fortnite|ShowSoundIndicatorLocalforTeam :Location (Transformation|GetActorLocation) :MaxAudibleDistance (Variables|/Script/Engine.SoundBase|GetMaxDistance (Variables|Default|GetExplodeSound)) :IndicatorType "Gunshot" :Team (Variables|Default|GetMyTeam) :Affiliations (Utilities|Array|MakeArray "Hostile" "Neutral") :Tint "(R=0.000000,G=0.000000,B=0.000000,A=0.000000)" :AttenuationMaxRangeMultiplier 1.0 :AttenuationIntensityMultiplier 1.0))
    (:then_1
      (bind _outactors (Collision|SphereOverlapActors _returnvalue_4 (ScalableFloat|GetValueatLevel (Variables|Default|GetDamageRadius)) (Utilities|Array|MakeArray "ObjectTypeQuery6" "ObjectTypeQuery1" "ObjectTypeQuery5" "ObjectTypeQuery3" "ObjectTypeQuery2" "ObjectTypeQuery4" "ObjectTypeQuery7") "/Script/Engine.Actor" (Utilities|Array|MakeArray self)))
      (bind _result (CallFunction|FilterbyLOS _outactors))
      (Utilities|FlowControl|Sequence
        (:then_0
          (bind _result_2 (CallFunction|GetAllPawns _result))
          (if (and (not (CallFunction|IsInInfiltrationMode)) (Class|AthenaFunctionLibrary|RowtoBool (Variables|Default|GetMarkPlayers)))
            (IndicatedManagement|AddActorsToIndicatedList _asfort_player_controller_athena _result_2 (IndicatedManagement|MakeIndicatedActorData :GroupIdentifier "Interrogation" :ParticleSystemData (IndicatedManagement|MakeIndicatedActorParticleSystemData "/Game/Creative/Items/Gameplay/Interrogate/P_Interrogate_Reveal.P_Interrogate_Reveal" "BoneSocketActor" "0, 0,100.000000" "0, 0,40.000000" "WorldOffset") :EventTags (GameplayTags|GetOwnedGameplayTags :!class /Script/GameplayTags.GameplayTagAssetInterface) :Duration _returnvalue_3 :StepTime 0.1 :IndicatorOffset "0.000000,0.000000,90.000000" :IndicatorDBNOOffset "0.000000,0.000000,30.000000" :ShareActorWith "SquadOnTeam" :Sound "/Game/Sounds/Fort_Human_Barks/SFX/Mang_Interrogate_Indicated_Cue.Mang_Interrogate_Indicated_Cue" :StateImageOverride "None" :StateIndexOverride -1 :bClampToScreen false :bHideStateImage false :bOnlyDisplayWhenStencilingIsDisabled false :bOnlyDisplayIfActorIsVisible false))
            (IndicatedManagement|AddActorsToStenciledList _asfort_player_controller_athena _result_2 (IndicatedManagement|MakeStenciledActorData :GroupIdentifier "Interrogation" :Duration _returnvalue_3 :StepTime 0.1 :ShareActorWith "SquadOnTeam" :FriendlyStencilName "None" :EnemyStencilName "RedOverlayID" :bOnlyDisplayIfActorIsVisible false))))
        (:then_1
          (Utilities|IsValid _returnvalue_2
            (:Is%20Valid
              (Variables|Default|SetDamageDealingAbilitySystemComponent _returnvalue_2))
            (:Is%20Not%20Valid
              (bind _asfort_ability_system_component (Utilities|Casting|CastToFortAbilitySystemComponent (Ability|GetAbilitySystemComponent self)))
              (Variables|Default|SetDamageDealingAbilitySystemComponent _asfort_ability_system_component)))
          (for _array_element _result
            (bind _asfort_pawn (Utilities|Casting|CastToFortPawn _array_element)
              (:then
                (if (or (Utilities|Equal(Object) _asfort_pawn (Variables|Actor|GetInstigator)) (Math|Byte|NotEqual(Byte) (Team|GetActorTeam _asfort_pawn) (Variables|Default|GetMyTeam)))
                  _))
              (:CastFailed
                _))
            (bind _returnvalue_5 (Utilities|MakeEffectContext))
            (GameplayEffects|ApplyGameplayEffectToTarget (Variables|Default|GetDamageDealingAbilitySystemComponent) (Variables|Default|GetExplosionGE) (Ability|GetAbilitySystemComponent _array_element) 0.0 _returnvalue_5)))))))

(event OnComponentEndOverlap(BeepBlinkTrigger)
  (params (OverlappedComponent :type Primitive%20Component%20Object%20Reference) (OtherActor :type Actor%20Object%20Reference) (OtherComp :type Primitive%20Component%20Object%20Reference) (OtherBodyIndex :type Integer))
  ;[1 1 1] Beep Logic
  (Networking|SwitchHasAuthority
    (:Authority
      (Utilities|Array|RemoveItem (Variables|Default|GetActorsInBeepRadius) OtherActor)
      (if (== (Utilities|Array|Length (Variables|Default|GetActorsInBeepRadius)) 0)
        (CallFunction|StopBeep)))))

(event Custom|DetachAndLaunch
  (bind (_returnvalue_x _returnvalue_y _returnvalue_z) (Math|Vector|BreakVector (Transformation|GetActorUpVector)))
  ;[1 1 1] Jump
  (Collision|SetActorEnableCollision false)
  (Variables|Default|SetBlockStoppingSim true)
  (CallFunction|RestartSimulation)
  (Variables|/Script/Engine.MovementComponent|SetVelocity (Math|Vector|vector*float (Math|Vector|MakeVector _returnvalue_x _returnvalue_y (Math|Float|Clamp(Float) _returnvalue_z)) 6e+02) (Variables|Default|GetProjectileMovement))
  (Utilities|FlowControl|Delay 0.1)
  (Collision|SetActorEnableCollision true)
  (Variables|Default|SetJumped true))

(event Custom|TriggerBeep
  (bind _beepparticles (Variables|Default|GetBeepParticles))
  ;[1 1 1] Beep Logic
  (Utilities|IsValid _beepparticles
    (:Is%20Valid
      (EventDispatchers|BindEventtoOnParticleSpawn _beepparticles (delegate Custom|Beep))
      (Effects|Components|ParticleSystem|SetEmitterEnable _beepparticles "Blink" true)
      (Effects|Components|ParticleSystem|SetEmitterEnable _beepparticles "Static" false)
      (Effects|Components|ParticleSystem|SetFloatParameter _beepparticles "SpawnRate" 4.0)
      (Effects|Components|ParticleSystem|SetFloatParameter _beepparticles "Lifetime" 0.15))))

(event Custom|TriggerJumpTimeline
  (bind _beepparticles (Variables|Default|GetBeepParticles))
  ;[1 1 1] Jump
  (Audio|Components|Audio|Play (Variables|Default|GetProximity%20Spin%20Sound) 0.0)
  (bind (_direction _spawnrate _lifetime _rotation_z _rotation_x) (Timeline|JumpOut Play)
    (:Update
      (Transformation|SetRelativeRotation (Variables|Default|GetMeshPivot) (Math|Rotator|MakeRotator (* _rotation_x 1.5e+02) 0.0 (* _rotation_z 7.2e+02)) false false)
      (Effects|Components|ParticleSystem|SetFloatParameter _beepparticles "SpawnRate" _spawnrate)
      (Effects|Components|ParticleSystem|SetFloatParameter _beepparticles "Lifetime" _lifetime))
    (:Finished
      (CallFunction|Explode))))

(event Custom|Beep
  (params (EventName :type Name) (EmitterTime :type Float%20%28single-precision%29) (Location :type Vector) (Velocity :type Vector))
  ;[1 1 1] Beep Logic
  (if (Utilities|Name|Equal(Name) EventName "Beep")
    (Audio|PlaySoundatLocation (Variables|Default|GetBeepSound) (Transformation|GetActorLocation))
    (Fortnite|ShowSoundIndicatorLocal :Location (Transformation|GetActorLocation) :MaxAudibleDistance (Variables|/Script/Engine.SoundBase|GetMaxDistance (Variables|Default|GetBeepSound)) :IndicatorType "Generic" :Instigator self :Tint "(R=0.000000,G=0.000000,B=0.000000,A=0.000000)" :OverrideIcon "/Game/UI/Foundation/Textures/Icons/HUD/T-Icon-Alert2-32.T-Icon-Alert2-32" :AttenuationMaxRangeMultiplier 1.0 :AttenuationIntensityMultiplier 1.0)))

(event Custom|ResetExplosionTimer
  (bind _beepparticles (Variables|Default|GetBeepParticles))
  ;[1 1 1] Jump
  (Timeline|JumpOut SetNewTime))

(event Custom|StopBeep
  (bind _beepparticles (Variables|Default|GetBeepParticles))
  ;[1 1 1] Beep Logic
  (Utilities|IsValid _beepparticles
    (:Is%20Valid
      (Effects|Components|ParticleSystem|SetEmitterEnable _beepparticles "Static" true)
      (Effects|Components|ParticleSystem|SetEmitterEnable _beepparticles "Blink" false)
      (EventDispatchers|UnbindEventfromOnParticleSpawn _beepparticles (delegate Custom|Beep)))))

(event Interaction|EventBlueprintOnInteract
  (params (InteractingPawn :type Fort%20Pawn%20Object%20Reference) (InteractionBeingAttempted :type EInteractionBeingAttempted%20Enum :default "FirstInteraction"))
  (bind _asfort_player_controller (Utilities|Casting|CastToFortPlayerController (Pawn|GetController InteractingPawn)))
  (Inventory|GiveItemtoInventoryOwner _asfort_player_controller (Variables|Default|GetItemToPickUpOnDisarm))
  (Variables|Default|SetIsInteract? true)
  (Actor|DestroyActor))

(event EventSpawnFXSounds
  (|CallParentFunction)
  (Audio|Components|Audio|FadeIn (Variables|Default|GetAmbientHumAudio) 0.5 1.0 0.0 "Linear" :!class /Script/Engine.AudioComponent))

(fn BlueprintCanInteract
  (params (InteractingPawn :type Fort%20Pawn%20Object%20Reference) (InteractionBeingAttempted :type EInteractionBeingAttempted%20Enum :default "FirstInteraction") (InteractionType :type TInteractionType%20Enum :default "IT_NoInteraction"))
  (return (and (not (GameplayTags|HasAnyTags (Class|FortGameStateAthena|GetAthenaPlaylistContextTags (Fortnite|Athena|GameState|GetGameStateAthena)) (LocalVariables|GetSpyLTMFactionTag))) (not (Variables|Default|GetJumped)))))

(fn BlueprintGetInteractionString
  (params (InteractingPawn :type Fort%20Pawn%20Object%20Reference) (InteractionBeingAttempted :type EInteractionBeingAttempted%20Enum :default "FirstInteraction"))
  (return (Variables|Default|GetInteractText)))

(fn BlueprintGetInteractionTime
  (params (InteractingPawn :type Fort%20Pawn%20Object%20Reference) (InteractionBeingAttempted :type EInteractionBeingAttempted%20Enum :default "FirstInteraction"))
  (return 0.0 true))

(fn CanBeSavedToCreativeVolume
  (return false))

(fn CheckLOSAgain
  (params (ActorToCheckLOSAgainst :type Actor%20Object%20Reference))
  (Networking|SwitchHasAuthority
    (:Authority
      (Utilities|CheckLOS ActorToCheckLOSAgainst 3e+01
        (:LOS
          (CallFunction|JumpIntoExplosion))))))

(fn FilterByLOS
  (params (_Array :type Array%20of%20Actor%20Object%20References))
  (for _array_element _Array
    (bind _asfort_pawn (Utilities|Casting|CastToFortPawn _array_element)
      (:then
        (Utilities|CheckLOS _asfort_pawn
          (:LOS
            _)))
      (:CastFailed
        _))
    (Utilities|Array|Add (LocalVariables|GetResult) _array_element))
  (return (LocalVariables|GetResult)))

(fn GetAllPawns
  (params (_Array :type Array%20of%20Actor%20Object%20References))
  (for _array_element _Array
    (Utilities|Casting|CastToFortPawn _array_element)
    (Utilities|Array|Add (LocalVariables|GetResult) _array_element))
  (return (LocalVariables|GetResult)))

(fn GetFocusedSocketLocation
  (return (+ (Transformation|GetActorLocation) "0, 0,30.000000")))

(fn Init
  (params (GravHitNormal :type Vector :default "0, 0, 0"))
  (|CallParentFunction GravHitNormal)
  (Transformation|SetActorRotation (Math|Rotator|MakeRotfromZ (Variables|Default|GetGravHitNormal)) true))

(fn IsinInfiltrationMode
  (return (GameplayTags|HasTag (Class|FortGameStateAthena|GetAthenaPlaylistContextTags (Fortnite|Athena|GameState|GetGameStateAthena)) "(TagName=\"Athena.Playlist.Infiltration\")")))

(fn Launch
  (params (LaunchSource :type Vector) (Amount :type Float%20%28double-precision%29))
  (bind _returnvalue (Transformation|GetActorLocation))
  (bind _returnvalue_2 (- _returnvalue LaunchSource))
  (bind _gravhitnormal (Variables|Default|GetGravHitNormal))
  (bind _returnvalue_3 (* _gravhitnormal Amount))
  (bind _returnvalue_4 (Math|Vector|Normalize _returnvalue_2))
  (bind _returnvalue_5 (* _returnvalue_4 Amount))
  (bind _returnvalue_6 (+ _returnvalue_5 _returnvalue_3))
  (bind _returnvalue_7 (Math|Vector|Normalize _returnvalue_6))
  (bind _returnvalue_8 (* _returnvalue_7 1e+02))
  (bind _returnvalue_9 (Transformation|GetActorLocation))
  (bind _returnvalue_10 (+ _returnvalue_9 _returnvalue_8))
  (CallFunction|RestartSimulation)
  (Transformation|DetachFromActor "KeepWorld" "KeepWorld" "KeepWorld")
  (bind (_outhit _returnvalue_11) (Collision|LineTraceByChannel _returnvalue_9 _returnvalue_10 "TraceTypeQuery10"))
  (Development|PrintString (Utilities|String|ToString(Vector) _returnvalue_8) true true "(R=1.000000,G=0.000000,B=0.847731,A=1.000000)" 8.0)
  (Transformation|SetActorLocation (Math|Vector|SelectVector (.Location (Collision|BreakHitResult _outhit)) _returnvalue_10 _returnvalue_11) false true)
  (Variables|/Script/Engine.MovementComponent|SetVelocity _returnvalue_6 (Variables|Default|GetProjectileMovement)))

(fn ConstructionScript
  (Networking|SwitchHasAuthority
    (:Authority
      (Variables|/Script/Engine.PrimitiveComponent|SetbGenerateOverlapEvents true (Variables|Default|GetExplosionTrigger))
      (Variables|/Script/Engine.PrimitiveComponent|SetbGenerateOverlapEvents true (Variables|Default|GetBeepBlinkTrigger))
      (|CallParentFunction))))
