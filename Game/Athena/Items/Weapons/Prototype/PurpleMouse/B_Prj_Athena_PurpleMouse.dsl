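; PurpleMouse prototype weapon projectile that spawns a team-aware, proximity-mine-style jumper actor.
; BeginPlay takes the team from a living instigator pawn unless SetTeam already ran, calling parent BeginPlay only when that pawn cast succeeds.
; The SpawnBGA override sends the parent-spawned actor to the instigator in a gameplay event for actor-limit tracking, then passes on the team.
; Spawns from water trigger JumpIntoExplosion on authority.
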(event Custom|SetTeam
  (params (Team :type Byte))
  (Variables|Default|SetTeam Team)
  (Variables|Default|SetTeamHasBeenSet true))

(event EventBeginPlay
  (if (not (Variables|Default|GetTeamHasBeenSet))
    (bind _asfort_pawn (Utilities|Casting|CastToFortPawn (Game|GetInstigator))
      (:then
        (if (not (FortPawn|IsDead _asfort_pawn))
          (CallFunction|SetTeam (Team|GetActorTeam _asfort_pawn)))
        (|CallParentFunction))
      (:CastFailed
        _))))

(event Collision|EventActorBeginOverlap
  (params (OtherActor :type Actor%20Object%20Reference)))

(event EventTick
  (params (DeltaSeconds :type Float%20%28single-precision%29)))

(fn SpawnBGA_
  (params (Normal :type Vector :default "0, 0, 0") (HitActor :type Actor%20Object%20Reference) (HitComp :type Primitive%20Component%20Object%20Reference) (Pos :type Vector :default "0, 0, 0") (FromWater :type Boolean) (Hit :type Hit%20Result%20Structure))
  (bind _returnvalue (|CallParentFunction Normal HitActor HitComp Pos FromWater))
  (Utilities|FlowControl|Sequence
    (:then_0
      (Ability|SendGameplayEventtoActor (Game|GetInstigator) "(TagName=\"Abilities.Drops.AddToActorLimit.ProxMine\")" (Utilities|Struct|MakeGameplayEventData :Target _returnvalue :ContextHandle "()" :InstigatorTags "(GameplayTags=)" :TargetTags "(GameplayTags=)" :EventMagnitude 0.0 :TargetData "()")) ; Calls event on player to trigger event on GA_Athena_Track_PurpleMouse
      (bind _asbga_athena_purple_mouse_jumper (Utilities|Casting|CastToBGA_Athena_PurpleMouse_Jumper _returnvalue))
      (if (Variables|Default|GetTeamHasBeenSet)
        (Class|BGAAthenaPurpleMouseJumper|SetMyTeam _asbga_athena_purple_mouse_jumper (Variables|Default|GetTeam) :!class /Game/Athena/Items/Weapons/Prototype/PurpleMouse/BGA_Athena_PurpleMouse_Jumper.BGA_Athena_PurpleMouse_Jumper_C)
        (else
          _))
      (if FromWater
        (Networking|SwitchHasAuthority
          (:Authority
            (Class|BGAAthenaPurpleMouseJumper|JumpIntoExplosion _asbga_athena_purple_mouse_jumper)))))
    (:then_1
      (return _returnvalue))))

(fn ConstructionScript
  (|CallParentFunction))
