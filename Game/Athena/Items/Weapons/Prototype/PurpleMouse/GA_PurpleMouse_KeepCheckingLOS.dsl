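; GA_PurpleMouse_KeepCheckingLOS is a gameplay ability for the PurpleMouse prototype weapon that repeatedly triggers line-of-sight rechecks on its avatar actor.
; On activation it starts a looping quarter-second timer.
; Each tick casts the granting effect's causer to BGA_Athena_PurpleMouse_Jumper and calls its CheckLOSAgain with the avatar, so the sight test itself lives in that actor.
; Ending the ability pauses the timer rather than clearing it.
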(event Ability|EventActivateAbility
  (bind _returnvalue (Utilities|Time|SetTimerbyEvent (delegate Custom|CheckLOSAgain) 0.25 true)))

(event Custom|CheckLOSAgain
  (bind (_returnvalue_owneractor _returnvalue_avataractor) (Utilities|Struct|BreakGameplayAbilityActorInfo (Ability|GetActorInfo)))
  (bind _asbga_athena_purple_mouse_jumper (Utilities|Casting|CastToBGA_Athena_PurpleMouse_Jumper (Ability|EffectContext|GetEffectCauser (Ability|GetGrantedbyEffectContext))))
  (Class|BGAAthenaPurpleMouseJumper|CheckLOSAgain _asbga_athena_purple_mouse_jumper _returnvalue_avataractor))

(event Ability|EventOnEndAbility
  (params (bWasCancelled :type Boolean))
  (Utilities|Time|PauseTimerbyHandle (stale Ability|EventActivateAbility _returnvalue)))

(event Ability|EventFailedToActivatePassiveAbility
  (params (ActorInfo :type Gameplay%20Ability%20Actor%20Info%20Structure))
  (|CallParentFunction))
